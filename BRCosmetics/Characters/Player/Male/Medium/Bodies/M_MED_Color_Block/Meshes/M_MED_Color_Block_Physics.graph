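# Physics asset for the M_MED_Color_Block male medium player body cosmetic, covering its dyn_ strap and banana bones plus the pelvis and right thigh.
# Capsules approximate the pelvis, right thigh and strap segment, while tiny spheres sit on the strap and banana master bones.
# The banana master links to the thigh, and the thigh and strap master link to the pelvis, all with free angular motion.
# Only the strap segment has tight swing and twist limits, against its master.
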PhysicsAsset /BRCosmetics/Characters/Player/Male/Medium/Bodies/M_MED_Color_Block/Meshes/M_MED_Color_Block_Physics
bodies {
 dyn_banana_MSTR [{shapeType=Sphere, radius=0.505, shapeName=dyn_banana_MSTR_sphere}]
 dyn_strap_01 [{
   shapeType=Capsule
   length=4.966
   radius=1.1137
   shapeName=dyn_strap_01_capsule
   center=<5.5719 0.9861 -0.4618>
   rotation=<rot -0.0018 180 129.9977>
  }]
 dyn_strap_MSTR [{shapeType=Sphere, radius=0.505, shapeName=dyn_strap_MSTR_sphere}]
 pelvis [{shapeType=Capsule, length=9.875, radius=6.1531, shapeName=pelvis_capsule, center=<0.7187 -1.026 -0.3078>, rotation=<rot 0.1339 -90 -0.4794>}]
 thigh_r [{shapeType=Capsule, length=29.8014, radius=8.0354, shapeName=thigh_r_capsule, center=<19.4074 0 0>, rotation=<rot 90 0 0>}]
}
constraints [
 {
  bone1Name=dyn_banana_MSTR
  bone2Name=thigh_r
  swing1LimitDegrees=45
  swing1Motion=Free
  swing2LimitDegrees=45
  swing2Motion=Free
  twistLimitDegrees=45
  twistMotion=Free
 }
 {
  bone1Name=dyn_strap_MSTR
  bone2Name=pelvis
  swing1LimitDegrees=45
  swing1Motion=Free
  swing2LimitDegrees=45
  swing2Motion=Free
  twistLimitDegrees=45
  twistMotion=Free
 }
 {bone1Name=dyn_strap_01, bone2Name=dyn_strap_MSTR, swing1LimitDegrees=30, swing2LimitDegrees=5, twistLimitDegrees=3}
 {
  bone1Name=thigh_r
  bone2Name=pelvis
  swing1LimitDegrees=45
  swing1Motion=Free
  swing2LimitDegrees=45
  swing2Motion=Free
  twistLimitDegrees=45
  twistMotion=Free
 }
]
