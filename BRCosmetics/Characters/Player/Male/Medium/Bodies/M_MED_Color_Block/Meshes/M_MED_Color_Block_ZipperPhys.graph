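# Physics asset for the zipper on the male medium Color Block body, giving its dynamic bones collision and joints for secondary motion.
# Three capsule bodies cover the pelvis and the two zipper bones, dyn_zipper_MSTR and dyn_zipper_m.
# The master zipper bone rotates freely against the pelvis, while dyn_zipper_m hangs off it with moderate swing limits and a very tight twist limit.
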PhysicsAsset /BRCosmetics/Characters/Player/Male/Medium/Bodies/M_MED_Color_Block/Meshes/M_MED_Color_Block_ZipperPhys
bodies {
 dyn_zipper_MSTR [{
   shapeType=Capsule
   length=1.1194
   radius=0.8199
   shapeName=dyn_zipper_MSTR_capsule
   center=<0.054 -0.2829 0.0286>
   rotation=<rot 8.2266 -90 -13.1937>
  }]
 dyn_zipper_m [{shapeType=Capsule, length=1.755, radius=0.63, shapeName=dyn_zipper_m_capsule, center=<2.54 0 0>, rotation=<rot 90 0 0>}]
 pelvis [{shapeType=Capsule, length=9.875, radius=6.1531, shapeName=pelvis_capsule, center=<0.7187 -1.026 -0.3078>, rotation=<rot 0.1339 -90 -0.4794>}]
}
constraints [
 {
  bone1Name=dyn_zipper_MSTR
  bone2Name=pelvis
  swing1LimitDegrees=45
  swing1Motion=Free
  swing2LimitDegrees=45
  swing2Motion=Free
  twistLimitDegrees=45
  twistMotion=Free
 }
 {bone1Name=dyn_zipper_m, bone2Name=dyn_zipper_MSTR, swing1LimitDegrees=45, swing2LimitDegrees=60, twistLimitDegrees=2}
]
