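# Physics asset for the Color Block male medium body's crouch jacket mesh, covering dynamic jacket, hand-wire and walkman bones.
# Capsules on pelvis, lower spine and thighs form the body volume.
# Jacket chains (front and side pairs, plus a three-box D chain) and hand wires hang from MSTR root bodies on spine_02; the walkman attaches to pelvis.
# Root and body joints rotate freely, while chain joints have swing and tight twist limits.
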PhysicsAsset /BRCosmetics/Characters/Player/Male/Medium/Bodies/M_MED_Color_Block/Meshes/M_MED_Color_Block_crouch_jacket
bodies {
 dyn_handWires [{shapeType=Sphere, radius=4.63, shapeName=dyn_handWires_sphere, center=<4.5142 0 -1.7439>}]
 dyn_handWires_MSTR [{shapeType=Sphere, radius=0.505, shapeName=dyn_handWires_MSTR_sphere}]
 dyn_jacket_D_01 [{shapeType=Box, extentX=9, extentY=15, extentZ=5.5, shapeName=dyn_jacket_D_01_box, center=<4.5 0 0>}]
 dyn_jacket_D_02 [{shapeType=Box, extentX=7.6, extentY=15, extentZ=5.5, shapeName=dyn_jacket_D_02_box, center=<4.25 0 0>}]
 dyn_jacket_D_03 [{shapeType=Box, extentX=6.25, extentY=15, extentZ=5.5, shapeName=dyn_jacket_D_03_box, center=<3.4 0 0>}]
 dyn_jacket_D_MSTR [{shapeType=Sphere, radius=0.505, shapeName=dyn_jacket_D_MSTR_sphere}]
 dyn_jacket_fr_l_01 [{shapeType=Capsule, length=9.9688, radius=3.125, center=<5.9331 -0.0003 4.5>, rotation=<rot 90 0 0>}]
 dyn_jacket_fr_l_01_MSTR [{shapeType=Capsule, length=1, radius=1}]
 dyn_jacket_fr_l_02 [{shapeType=Capsule, length=12.8438, radius=3.75, center=<8.8205 -1.7206 4.1582>, rotation=<rot 90 0 0>}]
 dyn_jacket_fr_r_01 [{shapeType=Capsule, length=7.0938, radius=3.125, center=<5.9331 -0.0003 4.5>, rotation=<rot 90 0 0>}]
 dyn_jacket_fr_r_01_MSTR [{shapeType=Capsule, length=1, radius=1}]
 dyn_jacket_fr_r_02 [{shapeType=Capsule, length=7.0938, radius=3.75, center=<5.9331 -0.0003 4.5>, rotation=<rot 90 0 0>}]
 dyn_jacket_l_01 [{shapeType=Capsule, length=7.3438, radius=4.75, center=<5.9331 -0.0003 4.5>, rotation=<rot 90 0 0>}]
 dyn_jacket_l_01_MSTR [{shapeType=Capsule, length=1, radius=1}]
 dyn_jacket_l_02 [{shapeType=Capsule, length=9.0938, radius=4.75, center=<6.6412 -0.0003 4.5>, rotation=<rot 90 0 0>}]
 dyn_jacket_r_01 [{shapeType=Capsule, length=7.0938, radius=5, center=<5.9331 -0.0003 4.5>, rotation=<rot 90 0 0>}]
 dyn_jacket_r_01_MSTR [{shapeType=Capsule, length=1, radius=1}]
 dyn_jacket_r_02 [{shapeType=Capsule, length=7.0938, radius=5, center=<5.9331 -0.0003 4.5>, rotation=<rot 90 0 0>}]
 dyn_walkman_MSTR [{
   shapeType=Box
   extentX=5.387
   extentY=10.2925
   extentZ=13.1123
   shapeName=dyn_walkman_MSTR_box
   center=<1.4159 0.8248 -0.6264>
   rotation=<rot 0 -114.9882 -10.0041>
  }]
 pelvis [{shapeType=Capsule, length=7.2188, radius=10.3125, center=<3.2501 0.1632 -0.0001>}]
 spine_01 [{shapeType=Capsule, length=7.2812, radius=10.25, center=<3.2501 0.7666 -0.0001>}]
 spine_02 [{shapeType=Capsule, length=7.7188, radius=9.8125, center=<3.2501 1.4404 -0.0002>}]
 thigh_l [{shapeType=Capsule, length=30.4062, radius=8.1562, center=<-17.1337 2.0212 -0.0001>, rotation=<rot 90 14.0363 16.8488>}]
 thigh_r [{shapeType=Capsule, length=29.7812, radius=8.7812, center=<17.2119 -0.43 0>, rotation=<rot 90 0 -179.9999>}]
}
constraints [
 {bone1Name=dyn_jacket_fr_l_02, bone2Name=dyn_jacket_fr_l_01, swing1LimitDegrees=25, swing2LimitDegrees=35, twistLimitDegrees=5}
 {bone1Name=dyn_jacket_l_02, bone2Name=dyn_jacket_l_01, swing1LimitDegrees=10, swing2LimitDegrees=45, twistLimitDegrees=5}
 {bone1Name=dyn_jacket_r_02, bone2Name=dyn_jacket_r_01, swing1LimitDegrees=10, swing2LimitDegrees=45, twistLimitDegrees=5}
 {bone1Name=dyn_jacket_fr_r_02, bone2Name=dyn_jacket_fr_r_01, swing1LimitDegrees=15, swing2LimitDegrees=20, twistLimitDegrees=5}
 {bone1Name=dyn_jacket_fr_l_01, bone2Name=dyn_jacket_fr_l_01_MSTR, swing1LimitDegrees=25, swing2LimitDegrees=35, twistLimitDegrees=5}
 {bone1Name=dyn_jacket_l_01, bone2Name=dyn_jacket_l_01_MSTR, swing1LimitDegrees=10, swing2LimitDegrees=45, twistLimitDegrees=5}
 {bone1Name=dyn_jacket_r_01, bone2Name=dyn_jacket_r_01_MSTR, swing1LimitDegrees=10, swing2LimitDegrees=45, twistLimitDegrees=5}
 {bone1Name=dyn_jacket_fr_r_01, bone2Name=dyn_jacket_fr_r_01_MSTR, swing1LimitDegrees=15, swing2LimitDegrees=20, twistLimitDegrees=5}
 {
  bone1Name=thigh_l
  bone2Name=pelvis
  swing1LimitDegrees=45
  swing1Motion=Free
  swing2LimitDegrees=45
  swing2Motion=Free
  twistLimitDegrees=45
  twistMotion=Free
 }
 {
  bone1Name=dyn_jacket_fr_r_01_MSTR
  bone2Name=spine_02
  swing1LimitDegrees=45
  swing1Motion=Free
  swing2LimitDegrees=45
  swing2Motion=Free
  twistLimitDegrees=45
  twistMotion=Free
 }
 {
  bone1Name=dyn_jacket_r_01_MSTR
  bone2Name=spine_02
  swing1LimitDegrees=45
  swing1Motion=Free
  swing2LimitDegrees=45
  swing2Motion=Free
  twistLimitDegrees=45
  twistMotion=Free
 }
 {
  bone1Name=dyn_jacket_fr_l_01_MSTR
  bone2Name=spine_02
  swing1LimitDegrees=45
  swing1Motion=Free
  swing2LimitDegrees=45
  swing2Motion=Free
  twistLimitDegrees=45
  twistMotion=Free
 }
 {
  bone1Name=dyn_jacket_l_01_MSTR
  bone2Name=spine_02
  swing1LimitDegrees=45
  swing1Motion=Free
  swing2LimitDegrees=45
  swing2Motion=Free
  twistLimitDegrees=45
  twistMotion=Free
 }
 {
  bone1Name=spine_01
  bone2Name=pelvis
  swing1LimitDegrees=45
  swing1Motion=Free
  swing2LimitDegrees=45
  swing2Motion=Free
  twistLimitDegrees=45
  twistMotion=Free
 }
 {
  bone1Name=spine_02
  bone2Name=spine_01
  swing1LimitDegrees=45
  swing1Motion=Free
  swing2LimitDegrees=45
  swing2Motion=Free
  twistLimitDegrees=45
  twistMotion=Free
 }
 {
  bone1Name=thigh_r
  bone2Name=pelvis
  swing1LimitDegrees=45
  swing1Motion=Free
  swing2LimitDegrees=45
  swing2Motion=Free
  twistLimitDegrees=45
  twistMotion=Free
 }
 {
  bone1Name=dyn_walkman_MSTR
  bone2Name=pelvis
  swing1LimitDegrees=45
  swing1Motion=Free
  swing2LimitDegrees=45
  swing2Motion=Free
  twistLimitDegrees=45
  twistMotion=Free
 }
 {
  bone1Name=dyn_jacket_D_MSTR
  bone2Name=spine_02
  swing1LimitDegrees=45
  swing1Motion=Free
  swing2LimitDegrees=45
  swing2Motion=Free
  twistLimitDegrees=45
  twistMotion=Free
 }
 {bone1Name=dyn_jacket_D_01, bone2Name=dyn_jacket_D_MSTR, swing1LimitDegrees=20, swing2LimitDegrees=35, twistLimitDegrees=5}
 {bone1Name=dyn_jacket_D_02, bone2Name=dyn_jacket_D_01, swing1LimitDegrees=20, swing2LimitDegrees=35, twistLimitDegrees=5}
 {bone1Name=dyn_jacket_D_03, bone2Name=dyn_jacket_D_02, swing1LimitDegrees=20, swing2LimitDegrees=35, twistLimitDegrees=5}
 {
  bone1Name=dyn_handWires_MSTR
  bone2Name=spine_02
  swing1LimitDegrees=45
  swing1Motion=Free
  swing2LimitDegrees=45
  swing2Motion=Free
  twistLimitDegrees=45
  twistMotion=Free
 }
 {bone1Name=dyn_handWires, bone2Name=dyn_handWires_MSTR, swing1LimitDegrees=45, swing2LimitDegrees=45, twistLimitDegrees=20}
]
