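# Post-process style cosmetic AnimBP for the M_MED_Color_Block player body, driving secondary motion on jacket, zipper, walkman, cord, kneepad and shoelaces.
# It copies the parent mesh pose, applies bone constraints, corrective offsets and AnimDynamics, then caches it.
# Crouch and skydive-float states select between RigidBody jacket setups, followed by zipper constraint and physics.
# The event graph derives corrective alphas from reference-pose bone deltas and weapon animation type.
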AnimBlueprint /BRCosmetics/Characters/Player/Male/Medium/Bodies/M_MED_Color_Block/Meshes/M_MED_Color_Block_AnimBP
target_skeleton=&/M_MED_Color_Block_Skeleton
graphs [
 AnimGraph {nodes [
   Anim Dynamics - Bone: dyn_zipper(Alpha=1) {
    bGravityOverrideInSimSpace=true
    simulationSpace=BoneRelative
    alphaBoolBlend {blendOption=Linear}
    boundBone {boneName=dyn_zipper}
    componentLinearAccScale=<0.1 0.1 0.1>
    physicsBodyDefinitions [{
      sphereCollisionRadius=0
      boundBone {boneName=dyn_zipper}
      boxExtents=<1 1 1>
      constraintSetup {angularLimitsMax=<10 10 10>, angularLimitsMin=<-10 -10 -10>, angularTarget=<0 0 0>}
      localJointOffset=<0 0 -1>
     }]
    relativeSpaceBone {boneName=dyn_zipper_MSTR}
   }
   n1 = RigidBody(ComponentPose=Local To Component(LocalPose=Use cached pose 'PreRB'())) {
    bClampLinearTranslationLimitToRefPose=true
    bOverrideWorldGravity=true
    overridePhysicsAsset=&/M_MED_Color_Block_Physics
    simulationSpace=BaseBoneSpace
    alphaBoolBlend {blendOption=Linear}
    baseBoneRef {boneName=root}
    componentAppliedLinearAccClamp=<3000 3000 3000>
    componentLinearAccScale=<0.4 0.4 0.4>
    componentLinearVelScale=<0.2 0.2 0.2>
    overrideWorldGravity {z=-980}
   }
   n2 = "Transform (Modify) Bone - Bone: dyn_walkman_MSTR"(ComponentPose=@n1, Alpha=Get L_Thigh_Rot() -> (L_Thigh_Rot:double), Translation="0,1,2", Rotation="10,0,-100", Scale="1,1,1") {
    rotationMode=BMM_Additive
    rotationSpace=BCS_BoneSpace
    translationMode=BMM_Additive
    translationSpace=BCS_BoneSpace
    alphaBoolBlend {blendOption=Linear}
    boneToModify {boneName=dyn_walkman_MSTR}
    rotation {pitch=10, roll=-100}
    translation {y=1, z=2}
   }
   n3 = RigidBody!(ComponentPose=@n2, Alpha, OverrideWorldGravity:Vector="0,0,-980", ExternalForce) -> (Pose) {
    bClampLinearTranslationLimitToRefPose=true
    bOverrideWorldGravity=true
    overridePhysicsAsset=&/M_MED_Color_Block_broc_jacket
    simulationSpace=BaseBoneSpace
    alphaBoolBlend {blendOption=Linear}
    baseBoneRef {boneName=spine_02}
    componentAppliedLinearAccClamp=<3000 3000 3000>
    componentLinearAccScale {x=0.3}
    componentLinearVelScale {x=0.3}
    overrideWorldGravity {z=-980}
    simSpaceSettings {
     maxAngularAcceleration=500
     maxAngularVelocity=3
     maxLinearAcceleration=400
     maxLinearVelocity=250
     velocityScaleZ=0
     worldAlpha=0.3
     externalLinearDragV=<0.1 0.1>
    }
   }
   n4 = "Constraint - dyn_jacket_fr_l_01 (P)"(ComponentPose=Local To Component(LocalPose=Use cached pose 'PreRB'()), Alpha=1, ConstraintWeights_0=1) {
    alphaBoolBlend {blendOption=Linear}
    boneToModify {boneName=dyn_jacket_fr_l_01}
    constraintSetup [{targetBone {boneName=pelvis}}]
    constraintWeights [1]
   }
   n5 = "Constraint - dyn_jacket_l_01_MSTR (P)"(ComponentPose=@n4, Alpha=1, ConstraintWeights_0=1) {
    alphaBoolBlend {blendOption=Linear}
    boneToModify {boneName=dyn_jacket_l_01_MSTR}
    constraintSetup [{targetBone {boneName=pelvis}}]
    constraintWeights [1]
   }
   n6 = "Constraint - dyn_jacket_fr_r_01 (P)"(ComponentPose=@n5, Alpha=0.5, ConstraintWeights_0=1) {
    alphaBoolBlend {blendOption=Linear}
    boneToModify {boneName=dyn_jacket_fr_r_01}
    constraintSetup [{targetBone {boneName=pelvis}}]
    constraintWeights [1]
   }
   n7 = "Constraint - dyn_jacket_r_01 (P)"(ComponentPose=@n6, Alpha=1, ConstraintWeights_0=1) {
    alphaBoolBlend {blendOption=Linear}
    boneToModify {boneName=dyn_jacket_r_01}
    constraintSetup [{targetBone {boneName=pelvis}}]
    constraintWeights [1]
   }
   n8 = "Constraint - dyn_jacket_fr_r_01_MSTR (R)"(ComponentPose=@n7, Alpha=0.4, ConstraintWeights_0=1) {
    alphaBoolBlend {blendOption=Linear}
    boneToModify {boneName=dyn_jacket_fr_r_01_MSTR}
    constraintSetup [{transformType=Rotation, targetBone {boneName=thigh_twist_01_r}}]
    constraintWeights [1]
   }
   n9 = "Constraint - dyn_jacket_fr_l_02 (P)"(ComponentPose=@n8, Alpha=0.5, ConstraintWeights_0=1) {
    alphaBoolBlend {blendOption=Linear}
    boneToModify {boneName=dyn_jacket_fr_l_02}
    constraintSetup [{targetBone {boneName=thigh_l}}]
    constraintWeights [1]
   }
   n10 = Get L_Thigh_Rot()
   n11 = "Transform (Modify) Bone - Bone: dyn_walkman_MSTR"(ComponentPose=@n9, Alpha=@n10, Translation="0,1,10", Rotation="45.000,-20.000,30.000", Scale="1,1,1") {
    rotationMode=BMM_Additive
    rotationSpace=BCS_BoneSpace
    translationMode=BMM_Additive
    translationSpace=BCS_BoneSpace
    alphaBoolBlend {blendOption=Linear}
    boneToModify {boneName=dyn_walkman_MSTR}
    rotation=<rot 44.9998 -19.9999 30.0003>
    translation {y=1, z=10}
   }
   n12 = "Transform (Modify) Bone - Bone: dyn_walkman_01"(ComponentPose=@n11, Alpha=@n10, Translation="0,0,0", Rotation="-60,50,0", Scale="1,1,1") {
    rotationMode=BMM_Additive
    rotationSpace=BCS_BoneSpace
    translationMode=BMM_Additive
    translationSpace=BCS_BoneSpace
    alphaBoolBlend {blendOption=Linear}
    boneToModify {boneName=dyn_walkman_01}
    rotation {pitch=-60, yaw=50}
   }
   n13 = "Transform (Modify) Bone - Bone: dyn_cord_MSTR"(ComponentPose=@n12, Alpha=1, Translation="-5,0,-2.5", Rotation="0,20,0", Scale="1,1,1") {
    rotationMode=BMM_Additive
    rotationSpace=BCS_BoneSpace
    translationMode=BMM_Additive
    translationSpace=BCS_BoneSpace
    alphaBoolBlend {blendOption=Linear}
    boneToModify {boneName=dyn_cord_MSTR}
    rotation {yaw=20}
    translation {x=-5, z=-2.5}
   }
   n14 = "Transform (Modify) Bone - Bone: dyn_strap_MSTR"(ComponentPose=@n13, Alpha=1, Translation="0,0,4", Rotation="0,0,0", Scale="1,1,1") {
    rotationMode=BMM_Additive
    rotationSpace=BCS_BoneSpace
    translationMode=BMM_Additive
    translationSpace=BCS_BoneSpace
    alphaBoolBlend {blendOption=Linear}
    boneToModify {boneName=dyn_strap_MSTR}
    translation {z=4}
   }
   n15 = "Transform (Modify) Bone - Bone: dyn_strap_MSTR"(ComponentPose=@n14, Alpha=Get Leg_R_Rot_Fwd() -> (Leg_R_Rot_Fwd:double), Translation="0,0,0", Rotation="0,25,-100", Scale="1,1,1") {
    rotationMode=BMM_Additive
    rotationSpace=BCS_BoneSpace
    translationMode=BMM_Additive
    translationSpace=BCS_BoneSpace
    alphaBoolBlend {blendOption=Linear}
    boneToModify {boneName=dyn_strap_MSTR}
    rotation {roll=-100, yaw=25}
   }
   n16 = Get Leg_L_Rot_Fwd() -> (Leg_L_Rot_Fwd:double)
   n17 = "Transform (Modify) Bone - Bone: dyn_jacket_fr_l_01_MSTR"(ComponentPose=@n15, Alpha=@n16, Translation="-1.001,-0.861,1.113", Rotation="6.802,-35,-90", Scale="1,1,1") {
    rotationMode=BMM_Additive
    rotationSpace=BCS_BoneSpace
    translationMode=BMM_Additive
    translationSpace=BCS_BoneSpace
    alphaBoolBlend {blendOption=Linear}
    alphaScaleBiasClamp {bInterpResult=true, interpSpeedDecreasing=5, interpSpeedIncreasing=5}
    boneToModify {boneName=dyn_jacket_fr_l_01_MSTR}
    rotation=<rot 6.802 -35 -90>
    translation=<-1.001 -0.861 1.113>
   }
   n18 = "Transform (Modify) Bone - Bone: dyn_jacket_l_01_MSTR"(ComponentPose=@n17, Alpha=@n16, Translation="0,0,0", Rotation="-39,23.971,-60", Scale="1,1,1") {
    rotationMode=BMM_Additive
    rotationSpace=BCS_BoneSpace
    translationMode=BMM_Additive
    translationSpace=BCS_BoneSpace
    alphaBoolBlend {blendOption=Linear}
    alphaScaleBiasClamp {bInterpResult=true, interpSpeedDecreasing=5, interpSpeedIncreasing=5}
    boneToModify {boneName=dyn_jacket_l_01_MSTR}
    rotation=<rot -39 23.971 -60>
   }
   n19 = Get Leg_R_Rot_Fwd()
   n20 = "Transform (Modify) Bone - Bone: dyn_jacket_fr_r_01_MSTR"(ComponentPose=@n18, Alpha=@n19, Translation="0,0,0", Rotation="30,50,0", Scale="1,1,1") {
    rotationMode=BMM_Additive
    rotationSpace=BCS_BoneSpace
    translationMode=BMM_Additive
    translationSpace=BCS_BoneSpace
    alphaBoolBlend {blendOption=Linear}
    alphaScaleBiasClamp {bInterpResult=true, interpSpeedDecreasing=5, interpSpeedIncreasing=5}
    boneToModify {boneName=dyn_jacket_fr_r_01_MSTR}
    rotation {pitch=30, yaw=50}
   }
   n21 = "Transform (Modify) Bone - Bone: dyn_jacket_r_01_MSTR"(ComponentPose=@n20, Alpha=@n19, Translation="0,0,0", Rotation="11.994,-24.070,-1.873", Scale="1,1,1") {
    rotationMode=BMM_Additive
    rotationSpace=BCS_BoneSpace
    translationMode=BMM_Additive
    translationSpace=BCS_BoneSpace
    alphaBoolBlend {blendOption=Linear}
    alphaScaleBiasClamp {bInterpResult=true, interpSpeedDecreasing=5, interpSpeedIncreasing=5}
    boneToModify {boneName=dyn_jacket_r_01_MSTR}
    rotation=<rot 11.994 -24.07 -1.873>
   }
   n22 = "Transform (Modify) Bone - Bone: dyn_jacket_fr_l_01_MSTR"(ComponentPose=@n21, Alpha=Get Leg_L_Rifle_Rot() -> (Leg_L_Rifle_Rot:double), Translation="6,-3,0", Rotation="0,0,0", Scale="1,1,1") {
    rotationMode=BMM_Additive
    rotationSpace=BCS_BoneSpace
    translationMode=BMM_Additive
    translationSpace=BCS_BoneSpace
    alphaBoolBlend {blendOption=Linear}
    boneToModify {boneName=dyn_jacket_fr_l_01_MSTR}
    translation=<6 -3>
   }
   n23 = "Transform (Modify) Bone - Bone: dyn_jacket_fr_l_01_MSTR"(ComponentPose=@n22, Alpha=Get Body_Rotation() -> (Body_Rotation:double), Translation="0,0,0", Rotation="0,0,-50", Scale="1,1,1") {
    rotationMode=BMM_Additive
    rotationSpace=BCS_BoneSpace
    translationMode=BMM_Additive
    translationSpace=BCS_BoneSpace
    alphaBoolBlend {blendOption=Linear}
    boneToModify {boneName=dyn_jacket_fr_l_01_MSTR}
    rotation {roll=-50}
   }
   n24 = RigidBody(ComponentPose=@n23, OverrideWorldGravity="0,0,-980") {
    bClampLinearTranslationLimitToRefPose=true
    bOverrideWorldGravity=true
    overridePhysicsAsset=&/M_MED_Color_Block_crouch_jacket
    simulationSpace=BaseBoneSpace
    alphaBoolBlend {blendOption=Linear}
    baseBoneRef {boneName=spine_02}
    componentAppliedLinearAccClamp=<3000 3000 3000>
    componentLinearAccScale {x=0.1}
    componentLinearVelScale {x=0.1}
    overrideWorldGravity {z=-980}
    simSpaceSettings {
     maxAngularAcceleration=500
     maxAngularVelocity=3
     maxLinearAcceleration=400
     maxLinearVelocity=250
     velocityScaleZ=0
     worldAlpha=0.3
     externalLinearDragV=<0.1 0.1>
    }
   }
   n25 = Blend Poses by bool(BlendPose_0=Component To Local(ComponentPose=@n24), BlendPose_1=Component To Local(ComponentPose=@n3), bActiveValue=Get bIsCrouching()) {blendType=Linear}
   n26 = RigidBody(ComponentPose=Local To Component(LocalPose=Use cached pose 'PreRB'()), Alpha=0.35, OverrideWorldGravity="0,0,-980") {
    bClampLinearTranslationLimitToRefPose=true
    bOverrideWorldGravity=true
    overridePhysicsAsset=&M_MED_Color_Block_broc_jacket
    simulationSpace=BaseBoneSpace
    alphaBoolBlend {blendOption=Linear}
    baseBoneRef {boneName=spine_02}
    componentAppliedLinearAccClamp=<3000 3000 3000>
    componentLinearAccScale {x=0.3}
    componentLinearVelScale {x=0.3}
    overrideWorldGravity {z=-980}
    simSpaceSettings {
     maxAngularAcceleration=500
     maxAngularVelocity=3
     maxLinearAcceleration=400
     maxLinearVelocity=250
     velocityScaleZ=0
     worldAlpha=0.3
     externalLinearDragV=<0.1 0.1>
    }
   }
   n27 = Component To Local(ComponentPose=@n26)
   n28 = Get bIsSkydiveFloating()
   n29 = "Constraint - dyn_zipper_MSTR (PPPPPPPP)"!(
    ComponentPose=Local To Component(LocalPose=Blend Poses by bool(BlendPose_0=@n27, BlendPose_1=@n25, bActiveValue=@n28) {blendType=Linear})
    Alpha=1
    ConstraintWeights_0=0.5314
    ConstraintWeights_1:float=0.0479
    ConstraintWeights_2:float=0.3264
    ConstraintWeights_3:float=0.0584
    ConstraintWeights_4:float=0.0099
    ConstraintWeights_5:float=0.0024
    ConstraintWeights_6:float=0.0011
    ConstraintWeights_7:float=0.0224
   ) -> (Pose) {
    alphaBoolBlend {blendOption=Linear}
    boneToModify {boneName=dyn_zipper_MSTR}
    constraintSetup [
     {targetBone {boneName=dyn_jacket_fr_l_02}}
     {targetBone {boneName=dyn_jacket_l_01}}
     {targetBone {boneName=dyn_jacket_l_02}}
     {targetBone {boneName=dyn_jacket_fr_l_01}}
     {targetBone {boneName=spine_01}}
     {targetBone {boneName=spine_02}}
     {targetBone {boneName=spine_03}}
     {targetBone {boneName=pelvis}}
    ]
    constraintWeights [1,1,1,1,1,1,1,1]
   }
   n30 = RigidBody!(ComponentPose=@n29, Alpha, ExternalForce) -> (Pose) {
    bClampLinearTranslationLimitToRefPose=true
    bOverrideWorldGravity=true
    overridePhysicsAsset=&/M_MED_Color_Block_ZipperPhys
    simulationSpace=BaseBoneSpace
    alphaBoolBlend {blendOption=Linear}
    baseBoneRef {boneName=dyn_zipper_MSTR}
    componentAppliedLinearAccClamp=<3000 3000 3000>
    componentLinearAccScale=<0.6 0.6 0.6>
    componentLinearVelScale=<0.3 0.3 0.3>
    overrideWorldGravity {z=-500}
   }
   Output Pose(Result=Component To Local(ComponentPose=@n30)) {layerGroup=None}
   n31 = "Transform (Modify) Bone - Bone: dyn_beltBack_MSTR"(
    ComponentPose=Local To Component(LocalPose=Copy Pose From Mesh() {bUseAttachedParent=true})
    Alpha=1
    Translation="0,1,0"
    Rotation="0,0,0"
    Scale="1,1,1"
   ) {
    translationMode=BMM_Additive
    translationSpace=BCS_BoneSpace
    alphaBoolBlend {blendOption=Linear}
    boneToModify {boneName=dyn_beltBack_MSTR}
    translation {y=1}
   }
   n32 = "Constraint - dyn_cord_MSTR (PPPPPP)"!(
    ComponentPose=@n31
    Alpha=1
    ConstraintWeights_0=0.0766
    ConstraintWeights_1:float=0.3086
    ConstraintWeights_2:float=0.3695
    ConstraintWeights_3:float=0.1966
    ConstraintWeights_4:float=0.0226
    ConstraintWeights_5:float=0.0262
   ) -> (Pose) {
    alphaBoolBlend {blendOption=Linear}
    boneToModify {boneName=dyn_cord_MSTR}
    constraintSetup [
     {targetBone {boneName=spine_01}}
     {targetBone {boneName=spine_02}}
     {targetBone {boneName=spine_03}}
     {targetBone {boneName=spine_04}}
     {targetBone {boneName=spine_05}}
     {targetBone {boneName=pelvis}}
    ]
    constraintWeights [1,1,1,1,1,1]
   }
   n33 = "Transform (Modify) Bone - Bone: dyn_cord_MSTR"(ComponentPose=@n32, Alpha=1, Translation="0,0,5", Rotation="0,0,0", Scale="1,1,1") {
    rotationMode=BMM_Additive
    rotationSpace=BCS_BoneSpace
    translationMode=BMM_Additive
    translationSpace=BCS_BoneSpace
    alphaBoolBlend {blendOption=Linear}
    boneToModify {boneName=dyn_cord_MSTR}
    translation {z=5}
   }
   n34 = "Transform (Modify) Bone - Bone: dyn_cord_MSTR"(ComponentPose=@n33, Alpha=Get Spine Rotation Alpha() -> (Spine Rotation Alpha:double), Translation="0,0,4", Rotation="0,0,0", Scale="1,1,1") {
    rotationMode=BMM_Additive
    rotationSpace=BCS_BoneSpace
    translationMode=BMM_Additive
    translationSpace=BCS_BoneSpace
    alphaBoolBlend {blendOption=Linear}
    boneToModify {boneName=dyn_cord_MSTR}
    translation {z=4}
   }
   n35 = "Transform (Modify) Bone - Bone: dyn_cord_MSTR"!(ComponentPose=@n34, bAlphaBoolEnabled:bool=True, Translation="0,-12,0", Rotation="0,0,0", Scale="1,1,1") -> (Pose) {
    alphaInputType=Bool
    translationMode=BMM_Additive
    translationSpace=BCS_BoneSpace
    alphaBoolBlend {blendOption=Linear}
    boneToModify {boneName=dyn_cord_MSTR}
    translation {y=-12}
   }
   n36 = "Constraint - dyn_jacket_D_MSTR (P)"(ComponentPose=@n35, Alpha=0.6, ConstraintWeights_0=1) {
    alphaBoolBlend {blendOption=Linear}
    boneToModify {boneName=dyn_jacket_D_MSTR}
    constraintSetup [{targetBone {boneName=pelvis}}]
    constraintWeights [1]
   }
   n37 = "Transform (Modify) Bone - Bone: dyn_walkman_MSTR"(ComponentPose=@n36, Alpha=1, Translation="-1,1,0", Rotation="0,0,0", Scale="1,1,1") {
    rotationMode=BMM_Additive
    rotationSpace=BCS_BoneSpace
    translationMode=BMM_Additive
    translationSpace=BCS_BoneSpace
    alphaBoolBlend {blendOption=Linear}
    boneToModify {boneName=dyn_walkman_MSTR}
    translation=<-1 1>
   }
   n38 = "Transform (Modify) Bone - Bone: dyn_walkman_MSTR"(ComponentPose=@n37, Alpha=Get Body Rot WAlkman() -> (Body Rot WAlkman:double), Translation="-1,-0.5,0", Rotation="5,0,0", Scale="1,1,1") {
    rotationMode=BMM_Additive
    rotationSpace=BCS_BoneSpace
    translationMode=BMM_Additive
    translationSpace=BCS_BoneSpace
    alphaBoolBlend {blendOption=Linear}
    boneToModify {boneName=dyn_walkman_MSTR}
    rotation {pitch=5}
    translation=<-1 -0.5>
   }
   n39 = "Constraint - dyn_banana_MSTR (PPP)"!(ComponentPose=@n38, Alpha=1, ConstraintWeights_0=0.0797, ConstraintWeights_1:float=0.0027, ConstraintWeights_2:float=0.9176) -> (Pose) {
    alphaBoolBlend {blendOption=Linear}
    boneToModify {boneName=dyn_banana_MSTR}
    constraintSetup [{targetBone {boneName=thigh_r}}, {targetBone {boneName=pelvis}}, {targetBone {boneName=thigh_twist_01_r}}]
    constraintWeights [1,1,1]
   }
   n40 = Anim Dynamics - Bone: dyn_banana(ComponentPose=@n39, Alpha=1, ExternalForce="0,0,0") {
    angularDampingOverride=0.999
    bGravityOverrideInSimSpace=true
    bOverrideAngularDamping=true
    bUseGravityOverride=true
    simulationSpace=RootRelative
    alphaBoolBlend {blendOption=Linear}
    boundBone {boneName=dyn_banana}
    componentAppliedLinearAccClamp=<3000 3000 3000>
    componentLinearAccScale=<0.3 0.3 0.3>
    componentLinearVelScale=<0.3 0.3 0.3>
    gravityOverride {z=-980}
    physicsBodyDefinitions [{
      sphereCollisionRadius=0
      boundBone {boneName=dyn_banana}
      boxExtents=<10 8 8>
      constraintSetup {angularLimitsMax=<20 45 45>, angularLimitsMin=<-45 -5 0>, angularTarget=<0 0 0>}
      localJointOffset=<1 0 0>
     }]
    relativeSpaceBone {boneName=pelvis}
   }
   n41 = "Constraint - dyn_shoelace_R_MSTR (PPP)"!(ComponentPose=@n40, Alpha=1, ConstraintWeights_0=0.0162, ConstraintWeights_1:float=0.1093, ConstraintWeights_2:float=0.8745) -> (Pose) {
    alphaBoolBlend {blendOption=Linear}
    boneToModify {boneName=dyn_shoelace_R_MSTR}
    constraintSetup [{targetBone {boneName=calf_twist_02_r}}, {targetBone {boneName=calf_twist_01_r}}, {targetBone {boneName=foot_r}}]
    constraintWeights [1,1,1]
   }
   n42 = "Constraint - dyn_shoelace_LR_MSTR (PPP)"!(ComponentPose=@n41, Alpha=1, ConstraintWeights_0=0.1709, ConstraintWeights_1:float=0.8142, ConstraintWeights_2:float=0.0149) -> (Pose) {
    alphaBoolBlend {blendOption=Linear}
    boneToModify {boneName=dyn_shoelace_LR_MSTR}
    constraintSetup [{targetBone {boneName=calf_twist_01_l}}, {targetBone {boneName=foot_l}}, {targetBone {boneName=calf_twist_02_l}}]
    constraintWeights [1,1,1]
   }
   n43 = "Constraint - dyn_shoelace_LL_MSTR (PP)"!(ComponentPose=@n42, Alpha=1, ConstraintWeights_0=0.1559, ConstraintWeights_1:float=0.8441) -> (Pose) {
    alphaBoolBlend {blendOption=Linear}
    boneToModify {boneName=dyn_shoelace_LL_MSTR}
    constraintSetup [{targetBone {boneName=calf_twist_01_l}}, {targetBone {boneName=foot_l}}]
    constraintWeights [1,1]
   }
   n44 = Anim Dynamics - Bone: dyn_shoelace_R_MSTR(ComponentPose=@n43, Alpha=1, ExternalForce="0,0,0") {
    angularDampingOverride=0.999
    bGravityOverrideInSimSpace=true
    bOverrideAngularDamping=true
    bUseGravityOverride=true
    simulationSpace=RootRelative
    alphaBoolBlend {blendOption=Linear}
    boundBone {boneName=dyn_shoelace_R_MSTR}
    componentAppliedLinearAccClamp=<3000 3000 3000>
    componentLinearAccScale=<0.3 0.3 0.3>
    componentLinearVelScale=<0.3 0.3 0.3>
    gravityOverride {z=-980}
    physicsBodyDefinitions [{
      sphereCollisionRadius=0
      boundBone {boneName=dyn_shoelace_R_MSTR}
      boxExtents=<10 8 8>
      constraintSetup {angularLimitsMax=<20 45 45>, angularLimitsMin=<-45 -45 -45>, angularTarget=<0 0 0>}
      localJointOffset=<0 0 -1>
     }]
    relativeSpaceBone {boneName=pelvis}
   }
   n45 = Anim Dynamics - Bone: dyn_shoelace_LL_MSTR(ComponentPose=@n44, Alpha=1, ExternalForce="0,0,0") {
    angularDampingOverride=0.999
    bGravityOverrideInSimSpace=true
    bOverrideAngularDamping=true
    bUseGravityOverride=true
    simulationSpace=RootRelative
    alphaBoolBlend {blendOption=Linear}
    boundBone {boneName=dyn_shoelace_LL_MSTR}
    componentAppliedLinearAccClamp=<3000 3000 3000>
    componentLinearAccScale=<0.3 0.3 0.3>
    componentLinearVelScale=<0.3 0.3 0.3>
    gravityOverride {z=-980}
    physicsBodyDefinitions [{
      sphereCollisionRadius=0
      boundBone {boneName=dyn_shoelace_LL_MSTR}
      boxExtents=<10 8 8>
      constraintSetup {angularLimitsMax=<45 10 45>, angularLimitsMin=<-45 -45 -45>, angularTarget=<0 0 0>}
      localJointOffset=<0 0 -1>
     }]
    relativeSpaceBone {boneName=pelvis}
   }
   n46 = Anim Dynamics - Bone: dyn_shoelace_LR_MSTR(ComponentPose=@n45, Alpha=1, ExternalForce="0,0,0") {
    angularDampingOverride=0.999
    bGravityOverrideInSimSpace=true
    bOverrideAngularDamping=true
    bUseGravityOverride=true
    simulationSpace=RootRelative
    alphaBoolBlend {blendOption=Linear}
    boundBone {boneName=dyn_shoelace_LR_MSTR}
    componentAppliedLinearAccClamp=<3000 3000 3000>
    componentLinearAccScale=<0.3 0.3 0.3>
    componentLinearVelScale=<0.3 0.3 0.3>
    gravityOverride {z=-980}
    physicsBodyDefinitions [{
      sphereCollisionRadius=0
      boundBone {boneName=dyn_shoelace_LR_MSTR}
      boxExtents=<10 8 8>
      constraintSetup {angularLimitsMax=<45 10 45>, angularLimitsMin=<-45 -45 -45>, angularTarget=<0 0 0>}
      localJointOffset=<0 0 -1>
     }]
    relativeSpaceBone {boneName=pelvis}
   }
   n47 = "Constraint - dyn_kneePad_MSTR (PPP)"!(ComponentPose=@n46, Alpha=1, ConstraintWeights_0=0.6341, ConstraintWeights_1:float=0.0126, ConstraintWeights_2:float=0.3533) -> (Pose) {
    alphaBoolBlend {blendOption=Linear}
    boneToModify {boneName=dyn_kneePad_MSTR}
    constraintSetup [{targetBone {boneName=thigh_r}}, {targetBone {boneName=calf_twist_02_r}}, {targetBone {boneName=calf_r}}]
    constraintWeights [1,1,1]
   }
   n48 = "Transform (Modify) Bone - Bone: dyn_kneePad_MSTR"(ComponentPose=@n47, Alpha=Get Kneepad Rot Alpha() -> (Kneepad Rot Alpha:double), Translation="-1,-1,0", Rotation="-5,-20,0", Scale="1,1,1") {
    rotationMode=BMM_Additive
    rotationSpace=BCS_BoneSpace
    translationMode=BMM_Additive
    translationSpace=BCS_BoneSpace
    alphaBoolBlend {blendOption=Linear}
    boneToModify {boneName=dyn_kneePad_MSTR}
    rotation {pitch=-5, yaw=-20}
    translation=<-1 -1>
   }
   n49 = "Transform (Modify) Bone - Bone: dyn_jacket_D_MSTR"!(ComponentPose=@n48, bAlphaBoolEnabled:bool=True, Translation="0,0,0", Rotation="0.000,0.000,30", Scale="1,1,1") -> (Pose) {
    alphaInputType=Bool
    rotationMode=BMM_Additive
    scaleMode=BMM_Additive
    translationMode=BMM_Additive
    alphaBoolBlend {blendOption=Linear}
    boneToModify {boneName=dyn_jacket_D_MSTR}
    rotation {roll=30}
   }
   n50 = "Transform (Modify) Bone - Bone: dyn_collar_MSTR"(ComponentPose=@n49, Alpha=1, Translation="-1.5,1,0", Rotation="0,0,0", Scale="1,1,1") {
    rotationMode=BMM_Additive
    rotationSpace=BCS_BoneSpace
    translationMode=BMM_Additive
    translationSpace=BCS_BoneSpace
    alphaBoolBlend {blendOption=Linear}
    boneToModify {boneName=dyn_collar_MSTR}
    translation=<-1.5 1>
   }
   n51 = "Transform (Modify) Bone - Bone: dyn_collar_MSTR"(
    ComponentPose=@n50
    Alpha=Get Neck_Rot_Alpha() -> (Neck_Rot_Alpha:double)
    Translation="-6,4,0"
    Rotation=Get Rotation() -> (Rotation:Rotator)
    Scale="1,1,1"
   ) {
    rotationMode=BMM_Additive
    rotationSpace=BCS_BoneSpace
    translationMode=BMM_Additive
    translationSpace=BCS_BoneSpace
    alphaBoolBlend {blendOption=Linear}
    boneToModify {boneName=dyn_collar_MSTR}
    rotation {yaw=-45}
    translation=<-6 4>
   }
   Save cached pose 'PreRB'(Pose=Component To Local(ComponentPose=@n51)) {cacheName=PreRB}
   Get bIsSkyDiving()
  ]}
 EventGraph {nodes [
   Event BlueprintUpdateAnimation()
   > n1 = Sequence!(execute) -> (then_0, then_1, then_2:Exec, then_3:Exec)
   n2 = BreakTransform(InTransform=GetDeltaTransformFromRefPose(self=GetOwnerSkeletalMeshComponent(), BoneName=thigh_l, BaseName=pelvis))
   > Set L_Thigh_Rot(L_Thigh_Rot:double=MapRangeClamped(Value=BreakRotator(InRot=@n2.Rotation).Yaw, InRangeA=-20, InRangeB=-105)) -> (Output_Get:double)
   n3 = GetOwnerSkeletalMeshComponent()
   > Set Leg_L_Rot_Fwd(Leg_L_Rot_Fwd:double=GetAlphaFromReferenceTransformDelta(SkeletalMeshComponent=@n3, BoneName=thigh_l, BaseName=pelvis, InputRange=(X=-20,Y=-100), TwistAxis="1,0,0")) -> (Output_Get:double)
   > Set Leg_R_Rot_Fwd(Leg_R_Rot_Fwd:double=GetAlphaFromReferenceTransformDelta(SkeletalMeshComponent=@n3, BoneName=thigh_r, BaseName=pelvis, InputRange=(X=-20,Y=-100), TwistAxis="1,0,0")) -> (Output_Get:double)
   n4 = BreakTransform(InTransform=GetDeltaTransformFromRefPose(self=GetOwnerSkeletalMeshComponent(), BoneName=calf_r, BaseName=thigh_r))
   Set Kneepad Rot Alpha(execute=@n1.then_1, Kneepad Rot Alpha:double=MapRangeClamped(Value=BreakRotator(InRot=@n4.Rotation).Yaw, InRangeA=20, InRangeB=105)) -> (Output_Get:double)
   n5 = GetDeltaTransformFromRefPose(self=GetOwnerSkeletalMeshComponent(), BoneName=spine_05, BaseName=pelvis)
   n6 = MapRangeClamped(Value=BreakRotator(InRot=BreakTransform(InTransform=@n5).Rotation), InRangeA=-37, InRangeB=-44)
   > Set Body Rot WAlkman(Body Rot WAlkman:double=@n6) -> (Output_Get:double)
   n7 = Cast To FortPlayerAnimInstance(execute=@n1.then_2, Object=GetAnimInstance(self=GetOwnerSkeletalMeshComponent()))
   > Set Leg_L_Rifle_Rot(Leg_L_Rifle_Rot:double=0) -> (Output_Get:double)
   > Set Body_Rotation(Body_Rotation:double=0) -> (Output_Get:double)
   > n8 = Switch on EFortWeaponCoreAnimation!(execute, Selection:EFortWeaponCoreAnimation Enum=Get WeaponCoreAnim(self=@n7)) -> (
    Melee:Exec
    Pistol:Exec
    Shotgun:Exec
    PaperBlueprint:Exec
    Rifle:Exec
    MeleeOneHand:Exec
    MachinePistol:Exec
    RocketLauncher:Exec
    GrenadeLauncher:Exec
    GoingCommando:Exec
    AssaultRifle:Exec
    TacticalShotgun:Exec
    SniperRifle:Exec
    TrapPlacement:Exec
    ShoulderLauncher:Exec
    AbilityDecoTool:Exec
    Crossbow:Exec
    C4:Exec
    RemoteControl:Exec
    DualWield:Exec
    AR_BullPup:Exec
    AR_ForwardGrip:Exec
    MedPackPaddles:Exec
    SMG_P90:Exec
    AR_DrumGun:Exec
    Consumable_Small:Exec
    Consumable_Large:Exec
    Balloon:Exec
    MountedTurret:Exec
    CreativeTool:Exec
    ExplosiveBow:Exec
    AshtonIndigo:Exec
    AshtonChicago:Exec
    MeleeDualWield:Exec
    Unarmed:Exec
   )
   Set Leg_L_Rifle_Rot(execute=[@n8.AssaultRifle, @n8.GoingCommando, @n8.Rifle], Leg_L_Rifle_Rot=1)
   > Set Body_Rotation(Body_Rotation=@n6)
   Set Spine Rotation Alpha(
    execute=@n1.then_3
    Spine Rotation Alpha:double=GetAlphaFromReferenceTransformDelta(SkeletalMeshComponent=GetOwnerSkeletalMeshComponent(), BoneName=spine_05, BaseName=pelvis, InputRange=(X=0,Y=20), TwistAxis="0,0,1")
   ) -> (Output_Get:double)
   n9 = GetOwnerSkeletalMeshComponent()
   n10 = BreakRotator(InRot=BreakTransform(InTransform=GetDeltaTransformFromRefPose(self=@n9, BoneName=head, BaseName=spine_05)).Rotation)
   > Set!(
    execute
    Rotation_Roll:float=float * float(A=@n10, B=0.5)
    Rotation_Pitch:float=float * float(A=@n10.Pitch, B=0.5)
    Rotation_Yaw:float=float * float(A=@n10.Yaw, B=0.6)
   ) -> (then, Output_Get:Rotator)
   > Set Neck_Rot_Alpha(Neck_Rot_Alpha:double=MapRangeClamped(Value=@n10.Yaw, InRangeA=0, InRangeB=-45)) -> (Output_Get:double)
   MapRangeClamped(Value=@n10.Yaw, InRangeA=8, InRangeB=-5)
  ]}
]
variables [
 Body Rot WAlkman {type=number}
 Body_Rotation {type=number}
 Kneepad Rot Alpha {type=number}
 L_Thigh_Rot {type=number}
 Leg_L_Rifle_Rot {type=number}
 Leg_L_Rot_Fwd {type=number}
 Leg_R_Rot_Fwd {type=number}
 Neck_Rot_Alpha {type=number}
 Rotation {type=Rotator}
 Spine Rotation Alpha {type=number}
]
